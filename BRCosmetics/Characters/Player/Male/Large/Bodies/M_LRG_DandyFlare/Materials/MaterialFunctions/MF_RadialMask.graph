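# Material function in the DandyFlare character cosmetic's folder that outputs a scalar radial mask of repeating directional wedges around a UV pivot.
# Inputs expose direction count, spread, gradient softness, base orientation and rotation.
# Pivot-centred UVs pass through a custom HLSL 2D rotation and atan2, giving a normalized angle.
# Fmod repeats that angle per segment, then a folded triangle ramp and a power curve shape the mask.
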MaterialFunction /BRCosmetics/Characters/Player/Male/Large/Bodies/M_LRG_DandyFlare/Materials/MaterialFunctions/MF_RadialMask
graph {nodes [
  MaterialExpressionNamedRerouteUsage_4 = NamedRerouteUsage()
  n1 = NamedRerouteUsage()
  n2 = NamedRerouteUsage()
  n3 = NamedRerouteUsage()
  MaterialExpressionMultiply_3 = Multiply(A=Add(A=Add(A=@n1, B=@n2) {constB=0.5}, B=Multiply(A=@n3) {constB=0.5}), B=NamedRerouteUsage())
  MaterialExpressionCustom_0 = Custom!(Dir=@MaterialExpressionNamedRerouteUsage_4, Angle=@MaterialExpressionMultiply_3) {
   code="return float2(\r\n    Dir.x * cos(Angle) + Dir.y * -sin(Angle),\r\n    Dir.x * sin(Angle) + Dir.y * cos(Angle)\r\n);"
   description=Rotate2D
   outputType=CMOT_Float2
   inputs [
    {
     inputName=Dir
     input {expression=&self:MaterialExpressionNamedRerouteUsage_4}
    }
    {
     inputName=Angle
     input {expression=&self:MaterialExpressionMultiply_3}
    }
   ]
  }
  MaterialExpressionConvert_0 = Convert(None=@MaterialExpressionCustom_0) {
   nodeName=Break Float2
   convertInputs [{
     type=Vector2
     expressionInput {
      expression=&self:MaterialExpressionCustom_0
      inputName=None
      mask=0
      maskA=0
      maskB=0
      maskG=0
      maskR=0
      outputIndex=0
     }
    }]
   convertOutputs [{type=Scalar}, {type=Scalar}]
  }
  n4 = Custom!(x=@MaterialExpressionConvert_0, y=@MaterialExpressionConvert_0) {
   code="return (atan2(y,x));"
   desc="[-Pi, Pi]"
   description=atan2
   outputType=CMOT_Float1
   inputs [
    {
     inputName=x
     input {expression=&self:MaterialExpressionConvert_0}
    }
    {
     inputName=y
     input {
      expression=&self:MaterialExpressionConvert_0
      outputIndex=1
     }
    }
   ]
  }
  n5 = Divide(A=@n4, B=NamedRerouteUsage()) {desc="[-0.5, 0.5]"}
  n6 = NamedRerouteUsage()
  n7 = Divide(A=Fmod(A=Add(A=@n5) {constB=0.5, desc="[0, 1]"}, B=Divide(B=@n6)), B=NamedRerouteUsage())
  n8 = Saturate(None=Abs(None=ConstantBiasScale(None=@n7) {bias=-0.5, scale=2}))
  n9 = NamedRerouteUsage()
  FunctionOutput(None=Power(Base=OneMinus(None=@n8), Exp=Multiply(A=@n9) {constB=10})) {outputName="Mask (S)"}
  n10 = FunctionInput() {
   bUsePreviewValueAsDefault=true
   description=Number of directions in the output mask
   inputName=Num Directions
   inputType=FunctionInput_Scalar
   sortPriority=20
   previewValue=<1 0 0 1>
  }
  NamedRerouteDeclaration(None=Max(A=@n10)) {name="Num Output Dir (S)", nodeColor=<rgba 0.2422 0 1 1>}
  n11 = FunctionInput() {
   bUsePreviewValueAsDefault=true
   description="Base orientation of the first output direction before rotation; 0: -U (left); 0.25: +V (down); 0.5: +U (right); 0.75: -V (top)"
   inputName=Orientation
   inputType=FunctionInput_Scalar
   sortPriority=50
  }
  NamedRerouteDeclaration(None=@n11) {name="Orientation (S)", nodeColor=<rgba 0 0.3478 1 1>}
  n12 = TextureCoordinate()
  n13 = FunctionInput() {bUsePreviewValueAsDefault=true, inputName=Pivot, inputType=FunctionInput_Vector2, sortPriority=10, previewValue=<0.5 0.5 0 1>}
  NamedRerouteDeclaration(None=Subtract(
    A=FunctionInput(Preview=@n12) {bUsePreviewValueAsDefault=true, description="Default: TexCoord[0]", inputName=UV, inputType=FunctionInput_Vector2}
    B=@n13
   )) {name="UV (V2)", nodeColor=<rgba 0 1 0.5781 1>}
  n14 = FunctionInput() {
   bUsePreviewValueAsDefault=true
   description="[0, 1]: zero to full circle"
   inputName=Spread
   inputType=FunctionInput_Scalar
   sortPriority=30
   previewValue=<0.125 0 0 1>
  }
  NamedRerouteDeclaration(None=@n14) {name="Spread (S)", nodeColor=<rgba 0.3828 1 0 1>}
  n15 = FunctionInput() {
   bUsePreviewValueAsDefault=true
   description="Amount of full revolutions; Counter-clockwise rotation"
   inputName=Rotation
   inputType=FunctionInput_Scalar
   sortPriority=60
  }
  NamedRerouteDeclaration(None=@n15) {name="Rotation (S)", nodeColor=<rgba 1 0 0.7031 1>}
  n16 = FunctionInput() {
   bUsePreviewValueAsDefault=true
   description="[0, 1] : hard to soft transition"
   inputName=Gradient
   inputType=FunctionInput_Scalar
   sortPriority=40
   previewValue=<0.2 0 0 1>
  }
  NamedRerouteDeclaration(None=@n16) {name="Gradient (S)", nodeColor=<rgba 0 1 0.9766 1>}
  NamedRerouteDeclaration(None=MaterialFunctionCall(Multiplier=Constant() {r=2}) {materialFunction=&Pi}.Result) {name="Tau (S)", nodeColor=<rgba 1 0.6562 0 1>}
 ]}
